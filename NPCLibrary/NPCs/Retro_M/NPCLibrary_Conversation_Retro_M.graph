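# Conversation database for the Retro_M NPC, a Heavy specialist, defining the greeting a player gets on talking to it in game.
# The entry point leads to one Speech node with a default flight-approach line and contextual lines for first-ever meetings, match conversation count and specific player outfits, weighted by priority.
# The speech then branches through link nodes to the shared NPCLibrary Quests and BasicServices conversations.
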ConversationDatabase /NPCLibrary/NPCs/Retro_M/NPCLibrary_Conversation_Retro_M
entry_points [{tag=Conversation.Entry.Retro_M, destinations [#1]}]
nodes [
 ConversationEntryPointNode_7 {
  id=#1
  class=ConversationEntryPointNode
  outputs [#2]
  properties {entryTag {tagName=Conversation.Entry.Retro_M}}
 }
 Speech_C_4 {
  id=#2
  class=Speech_C
  outputs [#3, #4]
  properties {
   nodeName=""
   generalConfig {
    defaultMessage=You're cleared for approach...
    contextualMessages [
     {
      message=So... this is land. Don't care for it.
      contextRequirements [{requirement {class=FortControllerRequirement_IsFirstEverConversationWithNPC}}]
      priority {curve {rowName=Tandem.Speech.Priority.FirstEverInteraction}}
     }
     {
      message=You're cleared for approach...
      contextRequirements [{requirement {
         class=FortControllerRequirement_CheckNPCMatchConversationCount
         properties {nPCUniqueIdTag {tagName=AISpawnerData.Type.Specialist.Heavy.Retro_M}}
        }}]
      priority {curve {rowName=None}}
     }
     {
      message="Stay clear of my flight path!"
      contextRequirements [{requirement {
         class=FortControllerRequirement_HasCID
         properties {allowedCIds [
           /BRCosmetics/Athena/Items/Cosmetics/Characters/Character_AirKringle
           /BRCosmetics/Athena/Items/Cosmetics/Characters/CID_359_Athena_Commando_F_Aztec
           /BRCosmetics/Athena/Items/Cosmetics/Characters/CID_437_Athena_Commando_F_AztecEclipse
           /BRCosmetics/Athena/Items/Cosmetics/Characters/Character_CorvidStomp
           /BRCosmetics/Athena/Items/Cosmetics/Characters/CID_017_Athena_Commando_M
           /BRCosmetics/Athena/Items/Cosmetics/Characters/Character_StormAviator
           /BRCosmetics/Athena/Items/Cosmetics/Characters/Character_MallardVantage
           /BRCosmetics/Athena/Items/Cosmetics/Characters/Character_MallardVantageFNCS
           /BRCosmetics/Athena/Items/Cosmetics/Characters/Character_CosmicSquatter
           /BRCosmetics/Athena/Items/Cosmetics/Characters/Character_FlamingoCountry
           /BRCosmetics/Athena/Items/Cosmetics/Characters/CID_464_Athena_Commando_M_Flamingo
           /BRCosmetics/Athena/Items/Cosmetics/Characters/CID_277_Athena_Commando_M_Moth
           /BRCosmetics/Athena/Items/Cosmetics/Characters/Character_FastCheetah
           /BRCosmetics/Athena/Items/Cosmetics/Characters/Character_SunnySquawk
           /BRCosmetics/Athena/Items/Cosmetics/Characters/Character_BuffBeak
           /BRCosmetics/Athena/Items/Cosmetics/Characters/Character_ClawedRaven
           /BRCosmetics/Athena/Items/Cosmetics/Characters/CID_744_Athena_Commando_F_DuckHero
           /BRCosmetics/Athena/Items/Cosmetics/Characters/CID_A_059_Athena_Commando_M_WickedDuck
           /BRCosmetics/Athena/Items/Cosmetics/Characters/CID_A_039_Athena_Commando_M_ChickenWarrior
           /BRCosmetics/Athena/Items/Cosmetics/Characters/CID_274_Athena_Commando_M_Feathers
          ]}
        }}]
      priority {value=2, curve {rowName=Tandem.Speech.Priority.OutfitReaction}}
     }
    ]
   }
  }
 }
 ConversationLinkNode_0 {id=#4, class=ConversationLinkNode, properties {remoteEntryTag {tagName=Conversation.Entry.BasicServices}}}
 ConversationLinkNode_1 {id=#3, class=ConversationLinkNode, properties {remoteEntryTag {tagName=Conversation.Entry.NPCLibrary.Quests}}}
]
